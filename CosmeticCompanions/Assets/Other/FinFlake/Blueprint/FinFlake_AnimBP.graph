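# Animation Blueprint for the FinFlake cosmetic companion, a swimming creature with idle, bored, combat, emote and reaction behaviours.
# A bool picks between cached low- and high-fidelity locomotion state machines, which share swim and idle poses, with the high path adding starts, turbo, additive antics and a bubble layer.
# Chooser proxy tables select stops, antics and poses.
# Worried, face, fin-override and turn layers follow, then a look-at aim offset and dead blending.
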AnimBlueprint /CosmeticCompanions/Assets/Other/FinFlake/Blueprint/FinFlake_AnimBP
target_skeleton=/CosmeticCompanions/Assets/Other/FinFlake/Meshes/FinFlake_Other_Skeleton
graphs [
 AnimGraph {nodes [
   n1 = Get CompanionDirection() -> (CompanionDirection:float)
   n2 = Blend Poses by bool(
    BlendPose_0=Use cached pose 'LowFidelityPose'()
    BlendPose_1=Use cached pose 'HighFidelityPose'()
    BlendTime_0=0.5
    BlendTime_1=0.5
    bActiveValue=Get bUseLowFidelity() -> (bUseLowFidelity:bool)
   ) {
    blendProfile=/CosmeticCompanions/Assets/Quadruped/SitPlant/Meshes/SitPlant_Quadruped_Skeleton:TimeBlend_FastFeetSlowBody
    blendType=Linear
    transitionType=Inertialization
   }
   n3 = "Sequence Player (additive) 'FinFlake_COR_Worried_ADD'"() {bLoopAnimation=false, sequence=/CosmeticCompanions/Assets/Other/FinFlake/Animation/Core/FinFlake_COR_Worried_ADD}
   n4 = Get bIsMoodWorried() -> (bIsMoodWorried:bool)
   n5 = Face_AnimLayer!(FaceInPose:Pose Link Structure=Slot 'DefaultSlot'(Source=Apply Additive!(Base=@n2, Additive=@n3, bAlphaBoolEnabled:bool=@n4) -> (Pose) {alphaInputType=Bool, alphaBoolBlend {blendOption=Linear}})) -> (Pose:Pose Link Structure) {is=title AnimGraphNode_LinkedAnimLayer, layer=Face_AnimLayer}
   n6 = AimOffset Player 'FinFlake_AO'!(
    X:float=Get LookAtYawAdjusted() -> (LookAtYawAdjusted:double)
    Y:float=Get LookAtPitchAdjusted() -> (LookAtPitchAdjusted:double)
    BasePose:Pose Link Structure=Apply Additive!(
     Base=ChildOverride_AnimLayer!(ParentPose:Pose Link Structure=@n5) -> (Pose:Pose Link Structure) {is=title AnimGraphNode_LinkedAnimLayer, layer=ChildOverride_AnimLayer}
     Additive=Blendspace Player 'FinFlake_COR_Loco_Turn_BS'!(X=@n1) -> (Pose) {blendSpace=/CosmeticCompanions/Assets/Other/FinFlake/Animation/Core/FinFlake_COR_Loco_Turn_BS}
     Alpha:float
    ) -> (Pose) {alphaBoolBlend {blendOption=Linear}}
    Alpha:float=Get AimOffsetAlpha() -> (AimOffsetAlpha:double)
   ) -> (Pose:Pose Link Structure) {
    blendSpace=/CosmeticCompanions/Assets/Other/FinFlake/Animation/AimOffsets/FinFlake_AO
    alphaBoolBlend {blendOption=Linear}
    alphaScaleBiasClamp {bInterpResult=true, interpSpeedDecreasing=6, interpSpeedIncreasing=3}
   }
   Output Pose(Result=Dead Blending!(Source:Pose Link Structure=@n6) -> (Pose:Pose Link Structure) {is=title AnimGraphNode_DeadBlending, defaultBlendMode=Linear})
   n7 = Get CompanionSpeed2D() -> (CompanionSpeed2D:float)
   AnimGraphNode_StateMachine_0 = LowFidelityLocomotion SM() {is=title AnimGraphNode_StateMachine}
   Save cached pose 'LowFidelityPose'(Pose=Apply Additive(
     Base=@AnimGraphNode_StateMachine_0
     Additive=Blendspace Player 'FinFlake_COR_Loco_Bubble_BS'!(X=@n7) -> (Pose) {blendSpace=/CosmeticCompanions/Assets/Other/FinFlake/Animation/Core/FinFlake_COR_Loco_Bubble_BS}
    ) {alphaBoolBlend {blendOption=Linear}}) {cacheName=LowFidelityPose}
   AnimGraphNode_StateMachine_4 = Antic Additives() {is=title AnimGraphNode_StateMachine, bAllowConduitEntryStates=true}
   AnimGraphNode_StateMachine_6 = Locomotion SM() {is=title AnimGraphNode_StateMachine}
   n8 = Apply Additive(Base=@AnimGraphNode_StateMachine_6, Additive=@AnimGraphNode_StateMachine_4) {alphaBoolBlend {blendOption=Linear}}
   n9 = Sequence Player 'FinFlake_COR_Base_Pose'() {bLoopAnimation=false, sequence=/CosmeticCompanions/Assets/Other/FinFlake/Animation/Core/FinFlake_COR_Base_Pose}
   Save cached pose 'HighFidelityPose'(Pose=Bubble_AnimLayer!(InPose:Pose Link Structure=Layered blend per bone(BasePose=@n8, BlendPoses_0=@n9) {layerSetup [{branchFilters [{boneName=face_attach}]}]}) -> (Pose:Pose Link Structure) {is=title AnimGraphNode_LinkedAnimLayer, layer=Bubble_AnimLayer}) {cacheName=HighFidelityPose}
   Save cached pose 'RunPose'(Pose=Blendspace Player 'BS_FinFlake_COR_Loco_Swim_BS'!(X=Get CompanionSpeed2D()) -> (Pose) {blendSpace=/CosmeticCompanions/Assets/Other/FinFlake/Animation/Core/BS_FinFlake_COR_Loco_Swim_BS}) {cacheName=RunPose}
   AnimGraphNode_StateMachine_1 = Idle SM() {is=title AnimGraphNode_StateMachine, bAllowConduitEntryStates=true}
   Save cached pose 'IdlePose'(Pose=@AnimGraphNode_StateMachine_1) {cacheName=IdlePose}
  ]}
 LowFidelityLocomotion SM {
  entry=Idle
  path=AnimGraph.AnimGraphNode_StateMachine_0.LowFidelityLocomotion SM
  ids=id
  nodes [
   AnimStateNode_0 = Run() {
    is=title AnimStateNode
    stateEntered {guid=#1}
    stateFullyBlended {guid=#2}
    stateLeft {guid=#3}
   }
   AnimStateNode_1 = Idle()
   AnimStateNode_4 = Stop() {is=title AnimStateNode}
   AnimStateTransitionNode_0 = AnimStateTransitionNode() {logicType=TLT_Inertialization, priorityOrder=2}
   AnimStateTransitionNode_1 = AnimStateTransitionNode() {crossfadeDuration=0.4, logicType=TLT_Inertialization}
   AnimStateTransitionNode_2 = AnimStateTransitionNode() {bAllowInertializationForSelfTransitions=true}
   AnimStateTransitionNode_25 = AnimStateTransitionNode() {logicType=TLT_Inertialization}
   AnimStateTransitionNode_26 = AnimStateTransitionNode() {automaticRuleTriggerTime=0, bAutomaticRuleBasedOnSequencePlayerInState=true}
  ]
  links [
   {
    dst=Run
    rule=AnimGraph.AnimGraphNode_StateMachine_0.LowFidelityLocomotion SM.AnimStateTransitionNode_0.Transition
    src=Idle
    via=AnimStateTransitionNode_0
   }
   {
    dst=Stop
    rule=AnimGraph.AnimGraphNode_StateMachine_0.LowFidelityLocomotion SM.AnimStateTransitionNode_1.Transition
    src=Run
    via=AnimStateTransitionNode_1
   }
   {
    dst=Idle
    rule=AnimGraph.AnimGraphNode_StateMachine_0.LowFidelityLocomotion SM.AnimStateTransitionNode_25.AnimationTransitionGraph_10
    src=Stop
    via=AnimStateTransitionNode_25
   }
   {
    dst=Idle
    rule=AnimGraph.AnimGraphNode_StateMachine_0.LowFidelityLocomotion SM.AnimStateTransitionNode_26.AnimationTransitionGraph_10
    src=Stop
    via=AnimStateTransitionNode_26
   }
   {
    dst=Run
    rule=AnimGraph.AnimGraphNode_StateMachine_0.LowFidelityLocomotion SM.AnimStateTransitionNode_2.Transition
    src=Stop
    via=AnimStateTransitionNode_2
   }
  ]
 }
 Run {
  path=AnimGraph.AnimGraphNode_StateMachine_0.LowFidelityLocomotion SM.AnimStateNode_0.Run
  nodes [Output Animation Pose(Result=Use cached pose 'RunPose'())]
 }
 Idle {
  path=AnimGraph.AnimGraphNode_StateMachine_0.LowFidelityLocomotion SM.AnimStateNode_1.Idle
  nodes [Output Animation Pose(Result=Use cached pose 'IdlePose'())]
 }
 Stop {
  path=AnimGraph.AnimGraphNode_StateMachine_0.LowFidelityLocomotion SM.AnimStateNode_4.Stop
  nodes [
   AnimGraphNode_ChooserPlayer_0 = Chooser Player: PRA_Companion_Stops!() -> (Pose:Pose Link Structure) {
    is=title AnimGraphNode_ChooserPlayer
    chooser {
     _structType=/Script/ProxyTable.LookupProxy
     proxy=/CosmeticCompanions/Chooser/PRA_Companion_Stops
     proxyTable {
      _structType=/Script/ProxyTable.ProxyTableContextProperty
      binding {contextIndex=0, displayName=Chooser Tables, isBoundToRoot=false, propertyBindingChain [ChooserTables]}
     }
    }
    chooserContextDefinition [
     {
      _structType=/Script/Chooser.ContextObjectTypeClass
      class=&/FinFlake_AnimBP.FinFlake_AnimBP_C
      direction=ReadWrite
     }
     {struct=/Script/Chooser.ChooserPlayerSettings, _structType=/Script/Chooser.ContextObjectTypeStruct, direction=Read}
    ]
   }
   Output Animation Pose(Result=@AnimGraphNode_ChooserPlayer_0)
  ]
 }
 AnimationBlendStackGraph_0 {
  path=AnimGraph.AnimGraphNode_StateMachine_0.LowFidelityLocomotion SM.AnimStateNode_4.Stop.AnimGraphNode_ChooserPlayer_0.AnimationBlendStackGraph_0
  nodes [Output Pose!(Result:Pose Link Structure=Blend Stack Input!() -> (Pose:Pose Link Structure) {is=title AnimGraphNode_BlendStackInput}) {is=title AnimGraphNode_BlendStackResult}]
 }
 Transition {
  path=AnimGraph.AnimGraphNode_StateMachine_0.LowFidelityLocomotion SM.AnimStateTransitionNode_0.Transition
  nodes [Result(bCanEnterTransition=Get bIsMoving() -> (bIsMoving:bool))]
 }
 Transition {
  path=AnimGraph.AnimGraphNode_StateMachine_0.LowFidelityLocomotion SM.AnimStateTransitionNode_1.Transition
  nodes [Result(bCanEnterTransition=Get bTriggerStop() -> (bTriggerStop:bool))]
 }
 Transition {
  path=AnimGraph.AnimGraphNode_StateMachine_0.LowFidelityLocomotion SM.AnimStateTransitionNode_2.Transition
  nodes [Result(bCanEnterTransition=Get bTriggerStopToStart() -> (bTriggerStopToStart:bool))]
 }
 AnimationTransitionGraph_10 {
  path=AnimGraph.AnimGraphNode_StateMachine_0.LowFidelityLocomotion SM.AnimStateTransitionNode_25.AnimationTransitionGraph_10
  nodes [Result(bCanEnterTransition=Get bIsMoodCombat() -> (bIsMoodCombat:bool))]
 }
 AnimationTransitionGraph_10 {
  path=AnimGraph.AnimGraphNode_StateMachine_0.LowFidelityLocomotion SM.AnimStateTransitionNode_26.AnimationTransitionGraph_10
  nodes [Result()]
 }
 Idle SM {
  entry=Conduit
  path=AnimGraph.AnimGraphNode_StateMachine_1.Idle SM
  ids=id
  nodes [
   IdleAlias() {is=title AnimStateAliasNode, also {aliases [Idle]}}
   BoredAlias() {is=title AnimStateAliasNode, also {aliases [Bored,BoredIntro]}}
   AnimStateConduitNode_0 = Conduit()
   AnimStateNode_0 = Idle()
   AnimStateNode_1 = Special Idle() {is=title AnimStateNode}
   AnimStateNode_2 = Special Exit Transition() {is=title AnimStateNode}
   AnimStateNode_3 = Antic() {is=title AnimStateNode}
   AnimStateNode_4 = Combat() {
    is=title AnimStateNode
    stateEntered {guid=#4}
    stateFullyBlended {guid=#5}
    stateLeft {guid=#6}
   }
   AnimStateNode_5 = Combat Exit Transition() {is=title AnimStateNode}
   AnimStateNode_6 = Bored() {is=title AnimStateNode}
   AnimStateNode_7 = BoredIntro() {is=title AnimStateNode}
   AnimStateNode_8 = BoredOutro() {is=title AnimStateNode}
   AnimStateTransitionNode_0 = AnimStateTransitionNode()
   AnimStateTransitionNode_1 = AnimStateTransitionNode()
   AnimStateTransitionNode_10 = AnimStateTransitionNode() {bAutomaticRuleBasedOnSequencePlayerInState=true}
   AnimStateTransitionNode_12 = AnimStateTransitionNode()
   AnimStateTransitionNode_13 = AnimStateTransitionNode() {automaticRuleTriggerTime=0, bAutomaticRuleBasedOnSequencePlayerInState=true}
   AnimStateTransitionNode_14 = AnimStateTransitionNode()
   AnimStateTransitionNode_15 = AnimStateTransitionNode() {bAutomaticRuleBasedOnSequencePlayerInState=true}
   AnimStateTransitionNode_2 = AnimStateTransitionNode() {blendMode=Sinusoidal, crossfadeDuration=0.3}
   AnimStateTransitionNode_3 = AnimStateTransitionNode() {
    automaticRuleTriggerTime=0
    bAutomaticRuleBasedOnSequencePlayerInState=true
    crossfadeDuration=0.5
    logicType=TLT_Inertialization
    blendProfileWrapper {blendProfile=&FinFlake_Other_Skeleton:SlowBubbleBlend}
   }
   AnimStateTransitionNode_4 = AnimStateTransitionNode() {automaticRuleTriggerTime=0.1, bAutomaticRuleBasedOnSequencePlayerInState=true, crossfadeDuration=0.1}
   AnimStateTransitionNode_5 = AnimStateTransitionNode() {logicType=TLT_Inertialization, priorityOrder=2}
   AnimStateTransitionNode_6 = AnimStateTransitionNode() {logicType=TLT_Inertialization}
   AnimStateTransitionNode_7 = AnimStateTransitionNode() {logicType=TLT_Inertialization, priorityOrder=4}
   AnimStateTransitionNode_8 = AnimStateTransitionNode() {priorityOrder=3}
   AnimStateTransitionNode_9 = AnimStateTransitionNode()
  ]
  links [
   {dst=Idle, rule=AnimGraph.AnimGraphNode_StateMachine_1.Idle SM.AnimStateTransitionNode_4.Transition, src=Antic, via=AnimStateTransitionNode_4}
   {
    dst=BoredOutro
    rule=AnimGraph.AnimGraphNode_StateMachine_1.Idle SM.AnimStateTransitionNode_14.Transition
    src=BoredAlias
    via=AnimStateTransitionNode_14
   }
   {
    dst=Bored
    rule=AnimGraph.AnimGraphNode_StateMachine_1.Idle SM.AnimStateTransitionNode_13.Transition
    src=BoredIntro
    via=AnimStateTransitionNode_13
   }
   {
    dst=Idle
    rule=AnimGraph.AnimGraphNode_StateMachine_1.Idle SM.AnimStateTransitionNode_15.Transition
    src=BoredOutro
    via=AnimStateTransitionNode_15
   }
   {
    dst=Combat Exit Transition
    rule=AnimGraph.AnimGraphNode_StateMachine_1.Idle SM.AnimStateTransitionNode_9.Transition
    src=Combat
    via=AnimStateTransitionNode_9
   }
   {
    dst=IdleAlias
    rule=AnimGraph.AnimGraphNode_StateMachine_1.Idle SM.AnimStateTransitionNode_10.Transition
    src=Combat Exit Transition
    via=AnimStateTransitionNode_10
   }
   {
    dst=Combat
    rule=AnimGraph.AnimGraphNode_StateMachine_1.Idle SM.AnimStateTransitionNode_8.Transition
    src=Conduit
    via=AnimStateTransitionNode_8
   }
   {
    dst=Idle
    rule=AnimGraph.AnimGraphNode_StateMachine_1.Idle SM.AnimStateTransitionNode_7.Transition
    src=Conduit
    via=AnimStateTransitionNode_7
   }
   {
    dst=Special Exit Transition
    rule=AnimGraph.AnimGraphNode_StateMachine_1.Idle SM.AnimStateTransitionNode_6.Transition
    src=Conduit
    via=AnimStateTransitionNode_6
   }
   {
    dst=Special Idle
    rule=AnimGraph.AnimGraphNode_StateMachine_1.Idle SM.AnimStateTransitionNode_5.Transition
    src=Conduit
    via=AnimStateTransitionNode_5
   }
   {dst=Antic, rule=AnimGraph.AnimGraphNode_StateMachine_1.Idle SM.AnimStateTransitionNode_1.Transition, src=Idle, via=AnimStateTransitionNode_1}
   {
    dst=BoredIntro
    rule=AnimGraph.AnimGraphNode_StateMachine_1.Idle SM.AnimStateTransitionNode_12.Transition
    src=Idle
    via=AnimStateTransitionNode_12
   }
   {
    dst=Conduit
    rule=AnimGraph.AnimGraphNode_StateMachine_1.Idle SM.AnimStateTransitionNode_0.Transition
    src=Idle
    via=AnimStateTransitionNode_0
   }
   {
    dst=Idle
    rule=AnimGraph.AnimGraphNode_StateMachine_1.Idle SM.AnimStateTransitionNode_3.Transition
    src=Special Exit Transition
    via=AnimStateTransitionNode_3
   }
   {
    dst=Special Exit Transition
    rule=AnimGraph.AnimGraphNode_StateMachine_1.Idle SM.AnimStateTransitionNode_2.Transition
    src=Special Idle
    via=AnimStateTransitionNode_2
   }
  ]
 }
 Conduit {path=AnimGraph.AnimGraphNode_StateMachine_1.Idle SM.AnimStateConduitNode_0.Conduit, nodes [Result(bCanEnterTransition=true)]}
 Idle {
  path=AnimGraph.AnimGraphNode_StateMachine_1.Idle SM.AnimStateNode_0.Idle
  nodes [Output Animation Pose(Result=Sequence Player 'FinFlake_COR_Idle_Lp'() {sequence=/CosmeticCompanions/Assets/Other/FinFlake/Animation/Core/FinFlake_COR_Idle_Lp}) {node {updateFunction {functionName=OnUpdate_IdleState}}, updateFunction {bSelfContext=true, memberName=OnUpdate_IdleState}}]
 }
 Special Idle {
  path=AnimGraph.AnimGraphNode_StateMachine_1.Idle SM.AnimStateNode_1.Special Idle
  nodes [
   AnimGraphNode_ChooserPlayer_1 = Chooser Player: PRA_Companion_SpecialIdle!() -> (Pose:Pose Link Structure) {
    is=title AnimGraphNode_ChooserPlayer
    chooser {
     _structType=&LookupProxy
     proxy=/CosmeticCompanions/Chooser/PRA_Companion_SpecialIdle
     proxyTable {
      _structType=&ProxyTableContextProperty
      binding {contextIndex=0, displayName=Chooser Tables, isBoundToRoot=false, propertyBindingChain [ChooserTables]}
     }
    }
    chooserContextDefinition [
     {
      _structType=&ContextObjectTypeClass
      class=&FinFlake_AnimBP_C
      direction=ReadWrite
     }
     {struct=&ChooserPlayerSettings, _structType=&ContextObjectTypeStruct, direction=Read}
    ]
    defaultSettings {blendOption=Linear}
   }
   Output Animation Pose(Result=Modify Curve(SourcePose=@AnimGraphNode_ChooserPlayer_1, CurveValues_0=1, Alpha=1) {curveNames [FaceCurveOverride], curveValues [0]})
  ]
 }
 AnimationBlendStackGraph_0 {
  path=AnimGraph.AnimGraphNode_StateMachine_1.Idle SM.AnimStateNode_1.Special Idle.AnimGraphNode_ChooserPlayer_1.AnimationBlendStackGraph_0
  nodes [Output Pose(Result=Blend Stack Input())]
 }
 Special Exit Transition {
  path=AnimGraph.AnimGraphNode_StateMachine_1.Idle SM.AnimStateNode_2.Special Exit Transition
  nodes [
   AnimGraphNode_ChooserPlayer_0 = Chooser Player: PRA_Companion_ReactToIdle!() -> (Pose:Pose Link Structure) {
    is=title AnimGraphNode_ChooserPlayer
    chooser {
     _structType=&LookupProxy
     proxy=/CosmeticCompanions/Chooser/PRA_Companion_ReactToIdle
     proxyTable {
      _structType=&ProxyTableContextProperty
      binding {contextIndex=0, displayName=Chooser Tables, isBoundToRoot=false, propertyBindingChain [ChooserTables]}
     }
    }
    chooserContextDefinition [
     {
      _structType=&ContextObjectTypeClass
      class=&FinFlake_AnimBP_C
      direction=ReadWrite
     }
     {struct=&ChooserPlayerSettings, _structType=&ContextObjectTypeStruct, direction=Read}
    ]
    defaultSettings {blendOption=Linear}
   }
   Output Animation Pose(Result=@AnimGraphNode_ChooserPlayer_0)
  ]
 }
 AnimationBlendStackGraph_0 {
  path=AnimGraph.AnimGraphNode_StateMachine_1.Idle SM.AnimStateNode_2.Special Exit Transition.AnimGraphNode_ChooserPlayer_0.AnimationBlendStackGraph_0
  nodes [Output Pose(Result=Blend Stack Input())]
 }
 Antic {
  path=AnimGraph.AnimGraphNode_StateMachine_1.Idle SM.AnimStateNode_3.Antic
  nodes [
   AnimGraphNode_ChooserPlayer_0 = Chooser Player: PRA_Companion_AnticsSelection!() -> (Pose:Pose Link Structure) {
    is=title AnimGraphNode_ChooserPlayer
    chooser {
     _structType=&LookupProxy
     proxy=/CosmeticCompanions/Chooser/PRA_Companion_AnticsSelection
     proxyTable {
      _structType=&ProxyTableContextProperty
      binding {contextIndex=0, displayName=Chooser Tables, isBoundToRoot=false, propertyBindingChain [ChooserTables]}
     }
    }
    chooserContextDefinition [
     {
      _structType=&ContextObjectTypeClass
      class=&FinFlake_AnimBP_C
      direction=ReadWrite
     }
     {struct=&ChooserPlayerSettings, _structType=&ContextObjectTypeStruct, direction=Read}
    ]
    defaultSettings {blendOption=Linear}
   }
   Output Animation Pose(Result=@AnimGraphNode_ChooserPlayer_0) {
    becomeRelevantFunction {bSelfContext=true, memberGuid=#7, memberName=ResetAnticFlag}
    node {becomeRelevantFunction {functionName=OnBecomeRelevant}}
   }
  ]
 }
 AnimationBlendStackGraph_0 {
  path=AnimGraph.AnimGraphNode_StateMachine_1.Idle SM.AnimStateNode_3.Antic.AnimGraphNode_ChooserPlayer_0.AnimationBlendStackGraph_0
  nodes [Output Pose(Result=Blend Stack Input())]
 }
 Combat {
  path=AnimGraph.AnimGraphNode_StateMachine_1.Idle SM.AnimStateNode_4.Combat
  nodes [
   n1 = Sequence Player 'FinFlake_COR_Angry_Idle_Lp'() {groupName=AngryLoop, method=SyncGroup, sequence=/CosmeticCompanions/Assets/Other/FinFlake/Animation/Reacts/FinFlake_COR_Angry_Idle_Lp}
   Output Animation Pose(Result=@n1)
  ]
 }
 Combat Exit Transition {
  path=AnimGraph.AnimGraphNode_StateMachine_1.Idle SM.AnimStateNode_5.Combat Exit Transition
  nodes [Output Animation Pose(Result=Sequence Player 'FinFlake_COR_Angry_Idle_Out'() {bLoopAnimation=false, sequence=/CosmeticCompanions/Assets/Other/FinFlake/Animation/Reacts/FinFlake_COR_Angry_Idle_Out})]
 }
 Bored {
  path=AnimGraph.AnimGraphNode_StateMachine_1.Idle SM.AnimStateNode_6.Bored
  nodes [Output Animation Pose(Result=Sequence Player 'FinFlake_COR_Bored_Lp'() {groupName=BoredLoop, method=SyncGroup, sequence=/CosmeticCompanions/Assets/Other/FinFlake/Animation/Core/FinFlake_COR_Bored_Lp})]
 }
 BoredIntro {
  path=AnimGraph.AnimGraphNode_StateMachine_1.Idle SM.AnimStateNode_7.BoredIntro
  nodes [Output Animation Pose(Result=Sequence Player 'FinFlake_COR_Bored_In'() {bLoopAnimation=false, sequence=/CosmeticCompanions/Assets/Other/FinFlake/Animation/Core/FinFlake_COR_Bored_In})]
 }
 BoredOutro {
  path=AnimGraph.AnimGraphNode_StateMachine_1.Idle SM.AnimStateNode_8.BoredOutro
  nodes [Output Animation Pose(Result=Sequence Player 'FinFlake_COR_Bored_Out'() {bLoopAnimation=false, sequence=/CosmeticCompanions/Assets/Other/FinFlake/Animation/Core/FinFlake_COR_Bored_Out})]
 }
 Transition {path=AnimGraph.AnimGraphNode_StateMachine_1.Idle SM.AnimStateTransitionNode_0.Transition, nodes [Result(bCanEnterTransition=true)]}
 Transition {
  path=AnimGraph.AnimGraphNode_StateMachine_1.Idle SM.AnimStateTransitionNode_1.Transition
  nodes [Result(bCanEnterTransition=Get bTriggerAntic() -> (bTriggerAntic:bool))]
 }
 Transition {path=AnimGraph.AnimGraphNode_StateMachine_1.Idle SM.AnimStateTransitionNode_10.Transition, nodes [Result()]}
 Transition {
  path=AnimGraph.AnimGraphNode_StateMachine_1.Idle SM.AnimStateTransitionNode_12.Transition
  nodes [Result(bCanEnterTransition=Get bIsMoodBored() -> (bIsMoodBored:bool))]
 }
 Transition {path=AnimGraph.AnimGraphNode_StateMachine_1.Idle SM.AnimStateTransitionNode_13.Transition, nodes [Result()]}
 Transition {
  path=AnimGraph.AnimGraphNode_StateMachine_1.Idle SM.AnimStateTransitionNode_14.Transition
  nodes [Result(bCanEnterTransition=NOT Boolean(A=Get bIsMoodBored()))]
 }
 Transition {path=AnimGraph.AnimGraphNode_StateMachine_1.Idle SM.AnimStateTransitionNode_15.Transition, nodes [Result()]}
 Transition {
  path=AnimGraph.AnimGraphNode_StateMachine_1.Idle SM.AnimStateTransitionNode_2.Transition
  nodes [Result(bCanEnterTransition=NOT Boolean(A=Get bIsPlayingEmote()))]
 }
 Transition {path=AnimGraph.AnimGraphNode_StateMachine_1.Idle SM.AnimStateTransitionNode_3.Transition, nodes [Result()]}
 Transition {path=AnimGraph.AnimGraphNode_StateMachine_1.Idle SM.AnimStateTransitionNode_4.Transition, nodes [Result()]}
 Transition {
  path=AnimGraph.AnimGraphNode_StateMachine_1.Idle SM.AnimStateTransitionNode_5.Transition
  nodes [Result(bCanEnterTransition=Get bIsEmoteActive() -> (bIsEmoteActive:bool))]
 }
 Transition {
  path=AnimGraph.AnimGraphNode_StateMachine_1.Idle SM.AnimStateTransitionNode_6.Transition
  nodes [Result(bCanEnterTransition=Get bIsHappyReactionActive() -> (bIsHappyReactionActive:bool))]
 }
 Transition {path=AnimGraph.AnimGraphNode_StateMachine_1.Idle SM.AnimStateTransitionNode_7.Transition, nodes [Result(bCanEnterTransition=true)]}
 Transition {
  path=AnimGraph.AnimGraphNode_StateMachine_1.Idle SM.AnimStateTransitionNode_8.Transition
  nodes [Result(bCanEnterTransition=Get bIsMoodCombat())]
 }
 Transition {
  path=AnimGraph.AnimGraphNode_StateMachine_1.Idle SM.AnimStateTransitionNode_9.Transition
  nodes [Result(bCanEnterTransition=NOT Boolean(A=Get bIsMoodCombat()))]
 }
 Antic Additives {
  entry=Conduit
  path=AnimGraph.AnimGraphNode_StateMachine_4.Antic Additives
  ids=id
  nodes [
   AnimStateConduitNode_0 = Conduit()
   AnimStateNode_0 = Idle()
   AnimStateNode_1 = Locomotion() {is=title AnimStateNode}
   AnimStateTransitionNode_2 = AnimStateTransitionNode() {crossfadeDuration=1}
   AnimStateTransitionNode_3 = AnimStateTransitionNode() {blendMode=Linear, crossfadeDuration=1}
   AnimStateTransitionNode_4 = AnimStateTransitionNode() {logicType=TLT_Inertialization}
   AnimStateTransitionNode_5 = AnimStateTransitionNode() {logicType=TLT_Inertialization}
  ]
  links [
   {
    dst=Idle
    rule=AnimGraph.AnimGraphNode_StateMachine_4.Antic Additives.AnimStateTransitionNode_4.Transition
    src=Conduit
    via=AnimStateTransitionNode_4
   }
   {
    dst=Locomotion
    rule=AnimGraph.AnimGraphNode_StateMachine_4.Antic Additives.AnimStateTransitionNode_5.Transition
    src=Conduit
    via=AnimStateTransitionNode_5
   }
   {
    dst=Locomotion
    rule=AnimGraph.AnimGraphNode_StateMachine_4.Antic Additives.AnimStateTransitionNode_2.Transition
    src=Idle
    via=AnimStateTransitionNode_2
   }
   {
    dst=Idle
    rule=AnimGraph.AnimGraphNode_StateMachine_4.Antic Additives.AnimStateTransitionNode_3.Transition
    src=Locomotion
    via=AnimStateTransitionNode_3
   }
  ]
 }
 Conduit {path=AnimGraph.AnimGraphNode_StateMachine_4.Antic Additives.AnimStateConduitNode_0.Conduit, nodes [Result(bCanEnterTransition=true)]}
 Idle {
  path=AnimGraph.AnimGraphNode_StateMachine_4.Antic Additives.AnimStateNode_0.Idle
  nodes [
   AnimGraphNode_StateMachine_0 = Body Pose SM() {is=title AnimGraphNode_StateMachine}
   Output Animation Pose(Result=@AnimGraphNode_StateMachine_0)
  ]
 }
 Body Pose SM {
  entry=Default
  path=AnimGraph.AnimGraphNode_StateMachine_4.Antic Additives.AnimStateNode_0.Idle.AnimGraphNode_StateMachine_0.Body Pose SM
  ids=id
  nodes [
   AnimStateNode_0 = Default() {is=title AnimStateNode}
   AnimStateNode_1 = BodyPose() {is=title AnimStateNode}
   AnimStateTransitionNode_5 = AnimStateTransitionNode() {crossfadeDuration=1}
   AnimStateTransitionNode_6 = AnimStateTransitionNode() {logicType=TLT_Inertialization}
   AnimStateTransitionNode_7 = AnimStateTransitionNode() {bAutomaticRuleBasedOnSequencePlayerInState=true, logicType=TLT_Inertialization}
  ]
  links [
   {
    dst=Default
    rule=AnimGraph.AnimGraphNode_StateMachine_4.Antic Additives.AnimStateNode_0.Idle.AnimGraphNode_StateMachine_0.Body Pose SM.AnimStateTransitionNode_5.Transition
    src=BodyPose
    via=AnimStateTransitionNode_5
   }
   {
    dst=Default
    rule=AnimGraph.AnimGraphNode_StateMachine_4.Antic Additives.AnimStateNode_0.Idle.AnimGraphNode_StateMachine_0.Body Pose SM.AnimStateTransitionNode_7.Transition
    src=BodyPose
    via=AnimStateTransitionNode_7
   }
   {
    dst=BodyPose
    rule=AnimGraph.AnimGraphNode_StateMachine_4.Antic Additives.AnimStateNode_0.Idle.AnimGraphNode_StateMachine_0.Body Pose SM.AnimStateTransitionNode_6.Transition
    src=Default
    via=AnimStateTransitionNode_6
   }
  ]
 }
 Default {
  path=AnimGraph.AnimGraphNode_StateMachine_4.Antic Additives.AnimStateNode_0.Idle.AnimGraphNode_StateMachine_0.Body Pose SM.AnimStateNode_0.Default
  nodes [Output Animation Pose(Result=Additive Identity Pose())]
 }
 BodyPose {
  path=AnimGraph.AnimGraphNode_StateMachine_4.Antic Additives.AnimStateNode_0.Idle.AnimGraphNode_StateMachine_0.Body Pose SM.AnimStateNode_1.BodyPose
  nodes [
   AnimGraphNode_ChooserPlayer_1 = Chooser Player: PRA_Companion_BodyPose!() -> (Pose:Pose Link Structure) {
    is=title AnimGraphNode_ChooserPlayer
    chooser {
     _structType=&LookupProxy
     proxy=/CosmeticCompanions/Chooser/PRA_Companion_BodyPose
     proxyTable {
      _structType=&ProxyTableContextProperty
      binding {contextIndex=0, displayName=Chooser Tables, isBoundToRoot=false, propertyBindingChain [ChooserTables]}
     }
    }
    chooserContextDefinition [
     {
      _structType=&ContextObjectTypeClass
      class=&FinFlake_AnimBP_C
      direction=ReadWrite
     }
     {struct=&ChooserPlayerSettings, _structType=&ContextObjectTypeStruct, direction=Read}
    ]
    defaultSettings {blendOption=Linear}
   }
   Output Animation Pose(Result=@AnimGraphNode_ChooserPlayer_1)
  ]
 }
 AnimationBlendStackGraph_0 {
  path=AnimGraph.AnimGraphNode_StateMachine_4.Antic Additives.AnimStateNode_0.Idle.AnimGraphNode_StateMachine_0.Body Pose SM.AnimStateNode_1.BodyPose.AnimGraphNode_ChooserPlayer_1.AnimationBlendStackGraph_0
  nodes [Output Pose(Result=Blend Stack Input())]
 }
 Transition {
  path=AnimGraph.AnimGraphNode_StateMachine_4.Antic Additives.AnimStateNode_0.Idle.AnimGraphNode_StateMachine_0.Body Pose SM.AnimStateTransitionNode_5.Transition
  nodes [Result(bCanEnterTransition=NOT Boolean(A=Get bCanPlayPoseOffset() -> (bCanPlayPoseOffset:bool)))]
 }
 Transition {
  path=AnimGraph.AnimGraphNode_StateMachine_4.Antic Additives.AnimStateNode_0.Idle.AnimGraphNode_StateMachine_0.Body Pose SM.AnimStateTransitionNode_6.Transition
  nodes [Result(bCanEnterTransition=Get bTriggerPoseOffset() -> (bTriggerPoseOffset:bool))]
 }
 Transition {
  path=AnimGraph.AnimGraphNode_StateMachine_4.Antic Additives.AnimStateNode_0.Idle.AnimGraphNode_StateMachine_0.Body Pose SM.AnimStateTransitionNode_7.Transition
  nodes [Result()]
 }
 Locomotion {
  path=AnimGraph.AnimGraphNode_StateMachine_4.Antic Additives.AnimStateNode_1.Locomotion
  nodes [
   AnimGraphNode_ChooserPlayer_0 = Chooser Player: CHT_FinFlake_SwimAdditives!() -> (Pose:Pose Link Structure) {
    is=title AnimGraphNode_ChooserPlayer
    evaluationFrequency=OnLoop
    chooser {_structType=/Script/Chooser.EvaluateChooser, chooser=/CosmeticCompanions/Assets/Other/FinFlake/Animation/Choosers/CHT_FinFlake_SwimAdditives}
    chooserContextDefinition [
     {
      _structType=&ContextObjectTypeClass
      class=&FinFlake_AnimBP_C
      direction=ReadWrite
     }
     {struct=&ChooserPlayerSettings, _structType=&ContextObjectTypeStruct, direction=Read}
    ]
    defaultSettings {bForceLooping=true}
   }
   n1 = Two Way Blend(A=@AnimGraphNode_ChooserPlayer_0, B=Additive Identity Pose(), Alpha=Get MovementAdditivesAlpha() -> (MovementAdditivesAlpha:double)) {blendNode {alphaBoolBlend {blendOption=Linear}, alphaScaleBiasClamp {interpSpeedDecreasing=5, interpSpeedIncreasing=5}}}
   Output Animation Pose(Result=@n1) {node {updateFunction {functionName=OnUpdate_LocomotionAdditive}}, updateFunction {bSelfContext=true, memberName=OnUpdate_LocomotionAdditive}}
  ]
 }
 AnimationBlendStackGraph_0 {
  path=AnimGraph.AnimGraphNode_StateMachine_4.Antic Additives.AnimStateNode_1.Locomotion.AnimGraphNode_ChooserPlayer_0.AnimationBlendStackGraph_0
  nodes [Output Pose(Result=Blend Stack Input())]
 }
 Transition {
  path=AnimGraph.AnimGraphNode_StateMachine_4.Antic Additives.AnimStateTransitionNode_2.Transition
  nodes [Result(bCanEnterTransition=Get bIsMoving())]
 }
 Transition {
  path=AnimGraph.AnimGraphNode_StateMachine_4.Antic Additives.AnimStateTransitionNode_3.Transition
  nodes [Result(bCanEnterTransition=NOT Boolean(A=Get bIsMoving()))]
 }
 Transition {
  path=AnimGraph.AnimGraphNode_StateMachine_4.Antic Additives.AnimStateTransitionNode_4.Transition
  nodes [Result(bCanEnterTransition=NOT Boolean(A=Get bIsMoving()))]
 }
 Transition {
  path=AnimGraph.AnimGraphNode_StateMachine_4.Antic Additives.AnimStateTransitionNode_5.Transition
  nodes [Result(bCanEnterTransition=Get bIsMoving())]
 }
 Locomotion SM {
  entry=Idle
  path=AnimGraph.AnimGraphNode_StateMachine_6.Locomotion SM
  ids=id
  nodes [
   ToStopAlias() {is=title AnimStateAliasNode, also {aliases [Run,SlowStart,TurboBoost]}}
   AnimStateNode_0 = Run() {
    stateEntered {guid=#1}
    stateFullyBlended {guid=#2}
    stateLeft {guid=#3}
   }
   AnimStateNode_1 = SlowStart() {is=title AnimStateNode}
   AnimStateNode_3 = Idle()
   AnimStateNode_4 = Stop()
   AnimStateNode_7 = TurboBoost() {is=title AnimStateNode}
   AnimStateTransitionNode_0 = AnimStateTransitionNode() {logicType=TLT_Inertialization, priorityOrder=2}
   AnimStateTransitionNode_16 = AnimStateTransitionNode() {minTimeBeforeReentry=2.5}
   AnimStateTransitionNode_17 = AnimStateTransitionNode() {automaticRuleTriggerTime=1.8, bAutomaticRuleBasedOnSequencePlayerInState=true, crossfadeDuration=0.3, logicType=TLT_Inertialization}
   AnimStateTransitionNode_21 = AnimStateTransitionNode() {bDisabled=true}
   AnimStateTransitionNode_25 = AnimStateTransitionNode() {logicType=TLT_Inertialization}
   AnimStateTransitionNode_26 = AnimStateTransitionNode() {automaticRuleTriggerTime=0, bAutomaticRuleBasedOnSequencePlayerInState=true}
   AnimStateTransitionNode_27 = AnimStateTransitionNode()
   AnimStateTransitionNode_29 = AnimStateTransitionNode()
   AnimStateTransitionNode_4 = AnimStateTransitionNode() {automaticRuleTriggerTime=0, bAutomaticRuleBasedOnSequencePlayerInState=true, logicType=TLT_Inertialization}
  ]
  links [
   {
    dst=SlowStart
    rule=AnimGraph.AnimGraphNode_StateMachine_6.Locomotion SM.AnimStateTransitionNode_0.Transition
    src=Idle
    via=AnimStateTransitionNode_0
   }
   {
    dst=TurboBoost
    rule=AnimGraph.AnimGraphNode_StateMachine_6.Locomotion SM.AnimStateTransitionNode_16.Transition
    src=Run
    via=AnimStateTransitionNode_16
   }
   {
    dst=Run
    rule=AnimGraph.AnimGraphNode_StateMachine_6.Locomotion SM.AnimStateTransitionNode_4.Transition
    src=SlowStart
    via=AnimStateTransitionNode_4
   }
   {
    dst=Idle
    rule=AnimGraph.AnimGraphNode_StateMachine_6.Locomotion SM.AnimStateTransitionNode_25.AnimationTransitionGraph_10
    src=Stop
    via=AnimStateTransitionNode_25
   }
   {
    dst=Idle
    rule=AnimGraph.AnimGraphNode_StateMachine_6.Locomotion SM.AnimStateTransitionNode_26.AnimationTransitionGraph_10
    src=Stop
    via=AnimStateTransitionNode_26
   }
   {
    dst=SlowStart
    rule=AnimGraph.AnimGraphNode_StateMachine_6.Locomotion SM.AnimStateTransitionNode_27.Transition
    src=Stop
    via=AnimStateTransitionNode_27
   }
   {
    dst=Stop
    rule=AnimGraph.AnimGraphNode_StateMachine_6.Locomotion SM.AnimStateTransitionNode_29.Transition
    src=ToStopAlias
    via=AnimStateTransitionNode_29
   }
   {
    dst=Run
    rule=AnimGraph.AnimGraphNode_StateMachine_6.Locomotion SM.AnimStateTransitionNode_17.Transition
    src=TurboBoost
    via=AnimStateTransitionNode_17
   }
   {
    dst=Run
    rule=AnimGraph.AnimGraphNode_StateMachine_6.Locomotion SM.AnimStateTransitionNode_21.Transition
    src=TurboBoost
    via=AnimStateTransitionNode_21
   }
  ]
 }
 Run {
  path=AnimGraph.AnimGraphNode_StateMachine_6.Locomotion SM.AnimStateNode_0.Run
  nodes [Output Animation Pose(Result=Use cached pose 'RunPose'())]
 }
 SlowStart {
  path=AnimGraph.AnimGraphNode_StateMachine_6.Locomotion SM.AnimStateNode_1.SlowStart
  nodes [Output Animation Pose(Result=Sequence Player 'FinFlake_COR_Loco_Start'() {bLoopAnimation=false, sequence=/CosmeticCompanions/Assets/Other/FinFlake/Animation/Core/FinFlake_COR_Loco_Start})]
 }
 Idle {
  path=AnimGraph.AnimGraphNode_StateMachine_6.Locomotion SM.AnimStateNode_3.Idle
  nodes [Output Animation Pose(Result=Use cached pose 'IdlePose'())]
 }
 Stop {
  path=AnimGraph.AnimGraphNode_StateMachine_6.Locomotion SM.AnimStateNode_4.Stop
  nodes [
   AnimGraphNode_ChooserPlayer_0 = Chooser Player: PRA_Companion_Stops() {
    chooser {
     _structType=&LookupProxy
     proxy=&PRA_Companion_Stops
     proxyTable {
      _structType=&ProxyTableContextProperty
      binding {contextIndex=0, displayName=Chooser Tables, isBoundToRoot=false, propertyBindingChain [ChooserTables]}
     }
    }
    chooserContextDefinition [
     {
      _structType=&ContextObjectTypeClass
      class=&FinFlake_AnimBP_C
      direction=ReadWrite
     }
     {struct=&ChooserPlayerSettings, _structType=&ContextObjectTypeStruct, direction=Read}
    ]
   }
   Output Animation Pose(Result=@AnimGraphNode_ChooserPlayer_0)
  ]
 }
 AnimationBlendStackGraph_0 {
  path=AnimGraph.AnimGraphNode_StateMachine_6.Locomotion SM.AnimStateNode_4.Stop.AnimGraphNode_ChooserPlayer_0.AnimationBlendStackGraph_0
  nodes [Output Pose(Result=Blend Stack Input())]
 }
 TurboBoost {
  path=AnimGraph.AnimGraphNode_StateMachine_6.Locomotion SM.AnimStateNode_7.TurboBoost
  nodes [Output Animation Pose(Result=Sequence Player 'FinFlake_COR_Loco_StartFast'() {bLoopAnimation=false, sequence=/CosmeticCompanions/Assets/Other/FinFlake/Animation/Core/FinFlake_COR_Loco_StartFast})]
 }
 Transition {
  path=AnimGraph.AnimGraphNode_StateMachine_6.Locomotion SM.AnimStateTransitionNode_0.Transition
  nodes [Result(bCanEnterTransition=Get bIsMoving())]
 }
 Transition {
  path=AnimGraph.AnimGraphNode_StateMachine_6.Locomotion SM.AnimStateTransitionNode_16.Transition
  nodes [Result(bCanEnterTransition=Get bTriggerTurbo() -> (bTriggerTurbo:bool))]
 }
 Transition {path=AnimGraph.AnimGraphNode_StateMachine_6.Locomotion SM.AnimStateTransitionNode_17.Transition, nodes [Result()]}
 Transition {
  path=AnimGraph.AnimGraphNode_StateMachine_6.Locomotion SM.AnimStateTransitionNode_21.Transition
  nodes [Result(bCanEnterTransition=NOT Boolean(A=Get bIsMoving()))]
 }
 AnimationTransitionGraph_10 {
  path=AnimGraph.AnimGraphNode_StateMachine_6.Locomotion SM.AnimStateTransitionNode_25.AnimationTransitionGraph_10
  nodes [Result(bCanEnterTransition=Get bIsMoodCombat())]
 }
 AnimationTransitionGraph_10 {path=AnimGraph.AnimGraphNode_StateMachine_6.Locomotion SM.AnimStateTransitionNode_26.AnimationTransitionGraph_10, nodes [Result()]}
 Transition {
  path=AnimGraph.AnimGraphNode_StateMachine_6.Locomotion SM.AnimStateTransitionNode_27.Transition
  nodes [Result(bCanEnterTransition=Get bTriggerStopToStart())]
 }
 Transition {
  path=AnimGraph.AnimGraphNode_StateMachine_6.Locomotion SM.AnimStateTransitionNode_29.Transition
  nodes [Result(bCanEnterTransition=Get bTriggerStop())]
 }
 Transition {path=AnimGraph.AnimGraphNode_StateMachine_6.Locomotion SM.AnimStateTransitionNode_4.Transition, nodes [Result(bCanEnterTransition=true)]}
 Bubble_AnimLayer {nodes [
   n1 = Blendspace Player 'FinFlake_COR_Loco_Bubble_BS'(X=Get CompanionSpeed2D()) {blendSpace=&FinFlake_COR_Loco_Bubble_BS}
   n2 = Input Pose()
   n3 = Get Playrate() -> (Playrate:double)
   Output Pose(Result=Apply Additive(
     Base=Apply Additive(Base=@n2, Additive=@n1) {alphaBoolBlend {blendOption=Linear}}
     Additive="Sequence Player (additive) 'FinFlake_COR_IdleBubble_Lp_ADD'"!(PlayRate:float=@n3) -> (Pose) {sequence=/CosmeticCompanions/Assets/Other/FinFlake/Animation/Core/FinFlake_COR_IdleBubble_Lp_ADD}
     Alpha=Get MovementAdditivesAlpha()
    ) {alphaBoolBlend {blendOption=Linear}, alphaScaleBiasClamp {bias=1, scale=-1}}) {is=title AnimGraphNode_Root, layerGroup=Default Shared Group}
  ]}
 ChildOverride_AnimLayer {nodes [Output Pose(Result=Layered blend per bone(
     BasePose=Input Pose() {name=ParentPose}
     BlendPoses_0=Sequence Player 'FinFlake_COR_Base_Pose'() {sequence=&FinFlake_COR_Base_Pose}
     BlendWeights_0=Get ChildAdditiveAlpha() -> (ChildAdditiveAlpha:float)
    ) {layerSetup [{branchFilters [{boneName=fin_base_l_01}, {boneName=fin_base_r_01}]}]})]}
 EventGraph {links [], nodes []}
 Face_AnimLayer {nodes [
   AnimGraphNode_StateMachine_0 = FaceAntics SM() {is=title AnimGraphNode_StateMachine}
   AnimGraphNode_StateMachine_2 = FacePose SM() {is=title AnimGraphNode_StateMachine, bAllowConduitEntryStates=true}
   Output Pose(Result=Apply Additive(
     Base=Input Pose() {name=FaceInPose}
     Additive=Apply Additive(Base=@AnimGraphNode_StateMachine_2, Additive=@AnimGraphNode_StateMachine_0) {alphaBoolBlend {blendOption=Linear}}
     Alpha=Get FaceAdditiveAlpha() -> (FaceAdditiveAlpha:double)
    ) {alphaBoolBlend {blendOption=Linear}, alphaScaleBiasClamp {interpSpeedDecreasing=5}}) {layerGroup=Default Shared Group}
  ]}
 FaceAntics SM {
  entry=Default
  path=Face_AnimLayer.AnimGraphNode_StateMachine_0.FaceAntics SM
  ids=id
  nodes [
   AnimStateNode_0 = Default()
   AnimStateNode_1 = FaceAntic() {is=title AnimStateNode}
   AnimStateTransitionNode_2 = AnimStateTransitionNode() {bAllowInertializationForSelfTransitions=true}
   AnimStateTransitionNode_3 = AnimStateTransitionNode() {automaticRuleTriggerTime=0, bAllowInertializationForSelfTransitions=true, bAutomaticRuleBasedOnSequencePlayerInState=true}
  ]
  links [
   {
    dst=FaceAntic
    rule=Face_AnimLayer.AnimGraphNode_StateMachine_0.FaceAntics SM.AnimStateTransitionNode_2.Transition
    src=Default
    via=AnimStateTransitionNode_2
   }
   {
    dst=Default
    rule=Face_AnimLayer.AnimGraphNode_StateMachine_0.FaceAntics SM.AnimStateTransitionNode_3.Transition
    src=FaceAntic
    via=AnimStateTransitionNode_3
   }
  ]
 }
 Default {path=Face_AnimLayer.AnimGraphNode_StateMachine_0.FaceAntics SM.AnimStateNode_0.Default, nodes [Output Animation Pose()]}
 FaceAntic {
  path=Face_AnimLayer.AnimGraphNode_StateMachine_0.FaceAntics SM.AnimStateNode_1.FaceAntic
  nodes [
   AnimGraphNode_ChooserPlayer_1 = Chooser Player: PRA_Companion_FaceAntics!() -> (Pose:Pose Link Structure) {
    is=title AnimGraphNode_ChooserPlayer
    chooser {
     _structType=&LookupProxy
     proxy=/CosmeticCompanions/Chooser/PRA_Companion_FaceAntics
     proxyTable {
      _structType=&ProxyTableContextProperty
      binding {contextIndex=0, displayName=Chooser Tables, isBoundToRoot=false, propertyBindingChain [ChooserTables]}
     }
    }
    chooserContextDefinition [
     {
      _structType=&ContextObjectTypeClass
      class=&FinFlake_AnimBP_C
      direction=ReadWrite
     }
     {struct=&ChooserPlayerSettings, _structType=&ContextObjectTypeStruct, direction=Read}
    ]
   }
   Output Animation Pose(Result=@AnimGraphNode_ChooserPlayer_1)
  ]
 }
 AnimationBlendStackGraph_0 {
  path=Face_AnimLayer.AnimGraphNode_StateMachine_0.FaceAntics SM.AnimStateNode_1.FaceAntic.AnimGraphNode_ChooserPlayer_1.AnimationBlendStackGraph_0
  nodes [Output Pose(Result=Blend Stack Input()) {is=title AnimGraphNode_BlendStackResult}]
 }
 Transition {
  path=Face_AnimLayer.AnimGraphNode_StateMachine_0.FaceAntics SM.AnimStateTransitionNode_2.Transition
  nodes [Result(bCanEnterTransition=Break Companion Random Number Timer!(CompanionRandomNumberTimer:"Companion Random Number Timer Structure (by ref)"=Get AuxiliaryTimer2() -> (AuxiliaryTimer2:Companion Random Number Timer Structure)) -> (ValueInt:int, ValueFloat:float=~, bJustChanged:bool=~) {is=title K2Node_BreakStruct}.bJustChanged)]
 }
 Transition {path=Face_AnimLayer.AnimGraphNode_StateMachine_0.FaceAntics SM.AnimStateTransitionNode_3.Transition, nodes [Result()]}
 FacePose SM {
  entry=Conduit
  path=Face_AnimLayer.AnimGraphNode_StateMachine_2.FacePose SM
  ids=id
  nodes [
   AnimStateConduitNode_0 = Conduit()
   AnimStateNode_0 = Default()
   AnimStateNode_1 = Mood() {is=title AnimStateNode}
   AnimStateTransitionNode_5 = AnimStateTransitionNode()
   AnimStateTransitionNode_6 = AnimStateTransitionNode()
   AnimStateTransitionNode_7 = AnimStateTransitionNode()
   AnimStateTransitionNode_8 = AnimStateTransitionNode()
  ]
  links [
   {
    dst=Default
    rule=Face_AnimLayer.AnimGraphNode_StateMachine_2.FacePose SM.AnimStateTransitionNode_5.Transition
    src=Conduit
    via=AnimStateTransitionNode_5
   }
   {
    dst=Mood
    rule=Face_AnimLayer.AnimGraphNode_StateMachine_2.FacePose SM.AnimStateTransitionNode_6.Transition
    src=Conduit
    via=AnimStateTransitionNode_6
   }
   {
    dst=Conduit
    rule=Face_AnimLayer.AnimGraphNode_StateMachine_2.FacePose SM.AnimStateTransitionNode_7.Transition
    src=Default
    via=AnimStateTransitionNode_7
   }
   {
    dst=Conduit
    rule=Face_AnimLayer.AnimGraphNode_StateMachine_2.FacePose SM.AnimStateTransitionNode_8.Transition
    src=Mood
    via=AnimStateTransitionNode_8
   }
  ]
 }
 Conduit {path=Face_AnimLayer.AnimGraphNode_StateMachine_2.FacePose SM.AnimStateConduitNode_0.Conduit, nodes [Result(bCanEnterTransition=true)]}
 Default {
  path=Face_AnimLayer.AnimGraphNode_StateMachine_2.FacePose SM.AnimStateNode_0.Default
  nodes [
   AnimGraphNode_ChooserPlayer_0 = Chooser Player: PRA_Companion_FacePose!() -> (Pose:Pose Link Structure) {
    is=title AnimGraphNode_ChooserPlayer
    chooser {
     _structType=&LookupProxy
     proxy=/CosmeticCompanions/Chooser/PRA_Companion_FacePose
     proxyTable {
      _structType=&ProxyTableContextProperty
      binding {contextIndex=0, displayName=Chooser Tables, isBoundToRoot=false, propertyBindingChain [ChooserTables]}
     }
    }
    chooserContextDefinition [
     {
      _structType=&ContextObjectTypeClass
      class=&FinFlake_AnimBP_C
      direction=ReadWrite
     }
     {struct=&ChooserPlayerSettings, _structType=&ContextObjectTypeStruct, direction=Read}
    ]
    defaultSettings {blendOption=Linear}
   }
   Output Animation Pose(Result=@AnimGraphNode_ChooserPlayer_0)
  ]
 }
 AnimationBlendStackGraph_0 {
  path=Face_AnimLayer.AnimGraphNode_StateMachine_2.FacePose SM.AnimStateNode_0.Default.AnimGraphNode_ChooserPlayer_0.AnimationBlendStackGraph_0
  nodes [Output Pose(Result=Blend Stack Input())]
 }
 Mood {
  path=Face_AnimLayer.AnimGraphNode_StateMachine_2.FacePose SM.AnimStateNode_1.Mood
  nodes [
   AnimGraphNode_ChooserPlayer_0 = Chooser Player: PRA_Companion_FacePose() {
    evaluationFrequency=OnUpdate
    chooser {
     _structType=&LookupProxy
     proxy=&PRA_Companion_FacePose
     proxyTable {
      _structType=&ProxyTableContextProperty
      binding {contextIndex=0, displayName=Chooser Tables, isBoundToRoot=false, propertyBindingChain [ChooserTables]}
     }
    }
    chooserContextDefinition [
     {
      _structType=&ContextObjectTypeClass
      class=&FinFlake_AnimBP_C
      direction=ReadWrite
     }
     {struct=&ChooserPlayerSettings, _structType=&ContextObjectTypeStruct, direction=Read}
    ]
    defaultSettings {blendOption=Linear}
   }
   Output Animation Pose(Result=@AnimGraphNode_ChooserPlayer_0)
  ]
 }
 AnimationBlendStackGraph_0 {
  path=Face_AnimLayer.AnimGraphNode_StateMachine_2.FacePose SM.AnimStateNode_1.Mood.AnimGraphNode_ChooserPlayer_0.AnimationBlendStackGraph_0
  nodes [Output Pose(Result=Blend Stack Input())]
 }
 Transition {
  path=Face_AnimLayer.AnimGraphNode_StateMachine_2.FacePose SM.AnimStateTransitionNode_5.Transition
  nodes [Result(bCanEnterTransition=Get bIsMoodDefault() -> (bIsMoodDefault:bool))]
 }
 Transition {
  path=Face_AnimLayer.AnimGraphNode_StateMachine_2.FacePose SM.AnimStateTransitionNode_6.Transition
  nodes [Result(bCanEnterTransition=NOT Boolean(A=Get bIsMoodDefault()))]
 }
 Transition {path=Face_AnimLayer.AnimGraphNode_StateMachine_2.FacePose SM.AnimStateTransitionNode_7.Transition, nodes [Result(bCanEnterTransition=true)]}
 Transition {path=Face_AnimLayer.AnimGraphNode_StateMachine_2.FacePose SM.AnimStateTransitionNode_8.Transition, nodes [Result(bCanEnterTransition=true)]}
 ResetAnticFlag {nodes [
   ResetAnticFlag!() -> (then, Context:"Anim Update Context Structure (by ref)", Node:"Anim Node Reference Structure (by ref)") {is=title K2Node_FunctionEntry}
   > Set bAnticFlag(bAnticFlag:bool=false) -> (Output_Get:bool)
   > Set bTriggerAntic(bTriggerAntic:bool=false) -> (Output_Get:bool)
   > Return Node()
  ]}
]
variables [ChooserTables {type=ProxyTable, default=/CosmeticCompanions/Assets/Other/FinFlake/Animation/Choosers/PRT_FinFlake_Choosers}]
